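# Animation Blueprint for the MoonShock Wings player body, posing its carapace wings on top of the attached parent mesh's pose.
# Copied body pose passes through bone constraints, two Control Rigs and shoulder correctives into a cached BodyPose.
# A layered blend limits the wingsPosition state machine (locomotion leans, airborne, targeting, crouch, skydive poses) to wing bones, switched against BodyPose by isWearingCape.
# Final constraints and additive offsets adjust wings, backpack attachment and cape_fix.
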AnimBlueprint /BRCosmetics/Characters/Player/Male/Medium/Bodies/M_MED_MoonShock_Wings/Meshes/M_MED_MoonShock_Wings_AnimBP
target_skeleton=/Game/Characters/Player/Male/Medium/Bodies/M_MED_MoonShock_Wings/Meshes/M_MED_MoonShock_Wings_Skeleton
graphs [
 AnimGraph {nodes [
   AnimGraphNode_StateMachine_0 = wingsPosition() {is=title AnimGraphNode_StateMachine}
   n1 = Layered blend per bone(
    BasePose=Use cached pose 'BodyPose'()
    BlendPoses_0=Blend Poses by bool(
     BlendPose_0=Use cached pose 'BodyPose'()
     BlendPose_1=@AnimGraphNode_StateMachine_0
     bActiveValue=Get isWearingCape() -> (isWearingCape:bool)
    ) {blendType=Linear}
   ) {layerSetup [{branchFilters [{boneName=CarapaceWings_l}, {boneName=CarapaceWings_r}, {boneName=None}]}]}
   n2 = "Constraint - CarapaceWings_l_A (PP)"!(ComponentPose=Local To Component(LocalPose=@n1), Alpha=1, ConstraintWeights_0=0.8, ConstraintWeights_1:float=0.2) -> (Pose) {
    alphaBoolBlend {blendOption=Linear}
    boneToModify {boneName=CarapaceWings_l_A}
    constraintSetup [{targetBone {boneName=CarapaceWings_l_A}}, {targetBone {boneName=CarapaceWings_l}}]
    constraintWeights [1,1]
   }
   n3 = "Constraint - CarapaceWings_l_B (PP)"!(ComponentPose=@n2, Alpha=1, ConstraintWeights_0=0.7, ConstraintWeights_1:float=0.3) -> (Pose) {
    alphaBoolBlend {blendOption=Linear}
    boneToModify {boneName=CarapaceWings_l_B}
    constraintSetup [{targetBone {boneName=CarapaceWings_l_B}}, {targetBone {boneName=CarapaceWings_l}}]
    constraintWeights [1,1]
   }
   n4 = "Constraint - CarapaceWings_r_A (PP)"!(ComponentPose=@n3, Alpha=1, ConstraintWeights_0=0.8, ConstraintWeights_1:float=0.2) -> (Pose) {
    alphaBoolBlend {blendOption=Linear}
    boneToModify {boneName=CarapaceWings_r_A}
    constraintSetup [{targetBone {boneName=CarapaceWings_r_A}}, {targetBone {boneName=CarapaceWings_r}}]
    constraintWeights [1,1]
   }
   n5 = "Constraint - CarapaceWings_r_B (PP)"!(ComponentPose=@n4, Alpha=1, ConstraintWeights_0=0.7, ConstraintWeights_1:float=0.3) -> (Pose) {
    alphaBoolBlend {blendOption=Linear}
    boneToModify {boneName=CarapaceWings_r_B}
    constraintSetup [{targetBone {boneName=CarapaceWings_r_B}}, {targetBone {boneName=CarapaceWings_r}}]
    constraintWeights [1,1]
   }
   n6 = "Constraint - attach_backpack (PPPPPP)"!(
    ComponentPose=@n5
    Alpha=1
    ConstraintWeights_0=0.1133
    ConstraintWeights_1:float=0.0579
    ConstraintWeights_2:float=0.3288
    ConstraintWeights_3:float=0.1133
    ConstraintWeights_4:float=0.0579
    ConstraintWeights_5:float=0.3288
   ) -> (Pose) {
    alphaBoolBlend {blendOption=Linear}
    boneToModify {boneName=attach_backpack}
    constraintSetup [
     {targetBone {boneName=CarapaceWings_r_B}}
     {targetBone {boneName=CarapaceWings_r_A}}
     {targetBone {boneName=CarapaceWings_r}}
     {targetBone {boneName=CarapaceWings_l_B}}
     {targetBone {boneName=CarapaceWings_l_A}}
     {targetBone {boneName=CarapaceWings_l}}
    ]
    constraintWeights [1,1,1,1,1,1]
   }
   n7 = "Transform (Modify) Bone - Bone: attach_backpack"(ComponentPose=@n6, Alpha=1, Translation="0,-4,0", Rotation="0,0,0", Scale="1,1,1") {
    rotationMode=BMM_Additive
    rotationSpace=BCS_BoneSpace
    translationMode=BMM_Additive
    translationSpace=BCS_BoneSpace
    alphaBoolBlend {blendOption=Linear}
    boneToModify {boneName=attach_backpack}
    translation {y=-4}
   }
   n8 = "Constraint - CarapaceWings_l (PP)"!(ComponentPose=@n7, Alpha=1, ConstraintWeights_0=0, ConstraintWeights_1:float=1) -> (Pose) {
    alphaBoolBlend {blendOption=Linear}
    boneToModify {boneName=CarapaceWings_l}
    constraintSetup [{targetBone {boneName=CarapaceWings_l}}, {targetBone {boneName=spine_05}}]
    constraintWeights [1,1]
   }
   n9 = "Constraint - CarapaceWings_r (PP)"!(ComponentPose=@n8, Alpha=1, ConstraintWeights_0=0, ConstraintWeights_1:float=1) -> (Pose) {
    alphaBoolBlend {blendOption=Linear}
    boneToModify {boneName=CarapaceWings_r}
    constraintSetup [{targetBone {boneName=CarapaceWings_r}}, {targetBone {boneName=spine_05}}]
    constraintWeights [1,1]
   }
   n10 = "Transform (Modify) Bone - Bone: CarapaceWings_l"(ComponentPose=@n9, Alpha=1, Translation="0,0,0", Rotation="-15,0,-5", Scale="0.7,1,0.7") {
    rotationMode=BMM_Additive
    rotationSpace=BCS_BoneSpace
    scaleMode=BMM_Additive
    translationMode=BMM_Additive
    translationSpace=BCS_BoneSpace
    alphaBoolBlend {blendOption=Linear}
    boneToModify {boneName=CarapaceWings_l}
    rotation {pitch=-15, roll=-5}
    scale {x=0.7, z=0.7}
   }
   n11 = "Transform (Modify) Bone - Bone: CarapaceWings_l_End_base"(ComponentPose=@n10, Alpha=1, Translation="0,0,30", Rotation="0,0,0", Scale="1,1,1") {
    rotationMode=BMM_Additive
    rotationSpace=BCS_BoneSpace
    translationMode=BMM_Additive
    translationSpace=BCS_BoneSpace
    alphaBoolBlend {blendOption=Linear}
    boneToModify {boneName=CarapaceWings_l_End_base}
    translation {z=30}
   }
   n12 = "Transform (Modify) Bone - Bone: CarapaceWings_r"(ComponentPose=@n11, Alpha=1, Translation="0,0,0", Rotation="-15,0,-5", Scale="0.7,1,0.7") {
    rotationMode=BMM_Additive
    rotationSpace=BCS_BoneSpace
    scaleMode=BMM_Additive
    translationMode=BMM_Additive
    translationSpace=BCS_BoneSpace
    alphaBoolBlend {blendOption=Linear}
    boneToModify {boneName=CarapaceWings_r}
    rotation {pitch=-15, roll=-5}
    scale {x=0.7, z=0.7}
   }
   n13 = "Transform (Modify) Bone - Bone: CarapaceWings_r_End_base"(ComponentPose=@n12, Alpha=1, Translation="0,0,-30", Rotation="0,0,0", Scale="1,1,1") {
    rotationMode=BMM_Additive
    rotationSpace=BCS_BoneSpace
    translationMode=BMM_Additive
    translationSpace=BCS_BoneSpace
    alphaBoolBlend {blendOption=Linear}
    boneToModify {boneName=CarapaceWings_r_End_base}
    translation {z=-30}
   }
   n14 = "Transform (Modify) Bone - Bone: CarapaceWings_l_B"(ComponentPose=@n13, Alpha=1, Translation="-2,-2,0", Rotation="-10,0,10", Scale="1,1,1") {
    rotationMode=BMM_Additive
    rotationSpace=BCS_BoneSpace
    translationMode=BMM_Additive
    translationSpace=BCS_BoneSpace
    alphaBoolBlend {blendOption=Linear}
    boneToModify {boneName=CarapaceWings_l_B}
    rotation {pitch=-10, roll=10}
    translation=<-2 -2>
   }
   n15 = "Transform (Modify) Bone - Bone: CarapaceWings_r_B"(ComponentPose=@n14, Alpha=1, Translation="2,2,0", Rotation="-10,0,10", Scale="1,1,1") {
    rotationMode=BMM_Additive
    rotationSpace=BCS_BoneSpace
    translationMode=BMM_Additive
    translationSpace=BCS_BoneSpace
    alphaBoolBlend {blendOption=Linear}
    boneToModify {boneName=CarapaceWings_r_B}
    rotation {pitch=-10, roll=10}
    translation=<2 2>
   }
   n16 = "Transform (Modify) Bone - Bone: CarapaceWings_l"(ComponentPose=@n15, Alpha=1, Translation="0,0,0", Rotation="0,0,10", Scale="1,1,1") {
    rotationMode=BMM_Additive
    rotationSpace=BCS_BoneSpace
    scaleMode=BMM_Additive
    translationMode=BMM_Additive
    translationSpace=BCS_BoneSpace
    alphaBoolBlend {blendOption=Linear}
    boneToModify {boneName=CarapaceWings_l}
    rotation {roll=10}
   }
   n17 = "Transform (Modify) Bone - Bone: CarapaceWings_r"(ComponentPose=@n16, Alpha=1, Translation="0,0,0", Rotation="0,0,10", Scale="1,1,1") {
    rotationMode=BMM_Additive
    rotationSpace=BCS_BoneSpace
    scaleMode=BMM_Additive
    translationMode=BMM_Additive
    translationSpace=BCS_BoneSpace
    alphaBoolBlend {blendOption=Linear}
    boneToModify {boneName=CarapaceWings_r}
    rotation {roll=10}
   }
   n18 = "Transform (Modify) Bone - Bone: cape_fix"(ComponentPose=@n17, Alpha=1, Translation="0,10,0", Rotation="0,0,0", Scale="1,1,1") {
    rotationMode=BMM_Additive
    rotationSpace=BCS_BoneSpace
    translationMode=BMM_Additive
    translationSpace=BCS_BoneSpace
    alphaBoolBlend {blendOption=Linear}
    boneToModify {boneName=cape_fix}
    translation {y=10}
   }
   Output Pose(Result=Component To Local(ComponentPose=@n18)) {layerGroup=None}
   n19 = "Constraint - attach_cape (PPP)"!(
    ComponentPose=Local To Component(LocalPose=Copy Pose From Mesh() {bUseAttachedParent=true})
    Alpha=1
    ConstraintWeights_0=0.7867
    ConstraintWeights_1:float=0.1575
    ConstraintWeights_2:float=0.0558
   ) -> (Pose) {
    alphaBoolBlend {blendOption=Linear}
    boneToModify {boneName=attach_cape}
    constraintSetup [{targetBone {boneName=spine_05}}, {targetBone {boneName=spine_03}}, {targetBone {boneName=clavicle_r}}]
    constraintWeights [1,1,1]
   }
   n20 = "Constraint - attach_backpack (PPPP)"!(
    ComponentPose=@n19
    Alpha=1
    ConstraintWeights_0=0.4248
    ConstraintWeights_1:float=0.3459
    ConstraintWeights_2:float=0.2283
    ConstraintWeights_3:float=0.001
   ) -> (Pose) {
    alphaBoolBlend {blendOption=Linear}
    boneToModify {boneName=attach_backpack}
    constraintSetup [{targetBone {boneName=spine_03}}, {targetBone {boneName=spine_05}}, {targetBone {boneName=spine_04}}, {targetBone {boneName=clavicle_l}}]
    constraintWeights [1,1,1,1]
   }
   n21 = "Constraint - tail_set (PPPPPPP)"!(
    ComponentPose=@n20
    Alpha=1
    ConstraintWeights_0=0.2501
    ConstraintWeights_1:float=0.248
    ConstraintWeights_2:float=0.1883
    ConstraintWeights_3:float=0.1558
    ConstraintWeights_4:float=0.1344
    ConstraintWeights_5:float=0.0119
    ConstraintWeights_6:float=0.0115
   ) -> (Pose) {
    alphaBoolBlend {blendOption=Linear}
    boneToModify {boneName=tail_set}
    constraintSetup [
     {targetBone {boneName=spine_02}}
     {targetBone {boneName=pelvis}}
     {targetBone {boneName=spine_03}}
     {targetBone {boneName=spine_01}}
     {targetBone {boneName=cape_fix}}
     {targetBone {boneName=deform_pec_aim_base_l}}
     {targetBone {boneName=spine_04}}
    ]
    constraintWeights [1,1,1,1,1,1,1]
   }
   n22 = Control Rig(Source=Component To Local(ComponentPose=@n21)) {
    alphaBoolBlend {blendOption=Linear}
    controlRigAssetReference {blueprintRigClass=&/M_MED_MoonShock_Wings_CtrlRig.M_MED_MoonShock_Wings_CtrlRig_C}
   }
   n23 = "Constraint - deform_upperarm_bk_l1 (P)"(ComponentPose=Local To Component(LocalPose=@n22), Alpha=1, ConstraintWeights_0=1) {
    alphaBoolBlend {blendOption=Linear}
    boneToModify {boneName=deform_upperarm_bk_l1}
    constraintSetup [{targetBone {boneName=deform_upperarm_bk_l}}]
    constraintWeights [1]
   }
   n24 = "Constraint - deform_upperarm_bk_r1 (P)"(ComponentPose=@n23, Alpha=1, ConstraintWeights_0=1) {
    alphaBoolBlend {blendOption=Linear}
    boneToModify {boneName=deform_upperarm_bk_r1}
    constraintSetup [{targetBone {boneName=deform_upperarm_bk_r}}]
    constraintWeights [1]
   }
   n25 = "Constraint - deform_upperarm_up_l1 (P)"(ComponentPose=@n24, Alpha=1, ConstraintWeights_0=1) {
    alphaBoolBlend {blendOption=Linear}
    boneToModify {boneName=deform_upperarm_up_l1}
    constraintSetup [{targetBone {boneName=deform_upperarm_up_l}}]
    constraintWeights [1]
   }
   n26 = "Constraint - deform_upperarm_up_r1 (P)"(ComponentPose=@n25, Alpha=1, ConstraintWeights_0=1) {
    alphaBoolBlend {blendOption=Linear}
    boneToModify {boneName=deform_upperarm_up_r1}
    constraintSetup [{targetBone {boneName=deform_upperarm_up_r}}]
    constraintWeights [1]
   }
   n27 = "Constraint - CarapaceWings_l_base (PPPPPP)"!(
    ComponentPose=@n26
    Alpha=1
    ConstraintWeights_0=0.2874
    ConstraintWeights_1:float=0.3784
    ConstraintWeights_2:float=0.129
    ConstraintWeights_3:float=0.1604
    ConstraintWeights_4:float=0.0226
    ConstraintWeights_5:float=0.0221
   ) -> (Pose) {
    alphaBoolBlend {blendOption=Linear}
    boneToModify {boneName=CarapaceWings_l_base}
    constraintSetup [
     {targetBone {boneName=spine_03}}
     {targetBone {boneName=clavicle_l}}
     {targetBone {boneName=spine_05}}
     {targetBone {boneName=spine_04}}
     {targetBone {boneName=upperarm_l}}
     {targetBone {boneName=deform_upperarm_bk_l1}}
    ]
    constraintWeights [1,1,1,1,1,1]
   }
   n28 = "Constraint - CarapaceWings_r_base (PPPPPP)"!(
    ComponentPose=@n27
    Alpha=1
    ConstraintWeights_0=0.2874
    ConstraintWeights_1:float=0.3784
    ConstraintWeights_2:float=0.129
    ConstraintWeights_3:float=0.1604
    ConstraintWeights_4:float=0.0226
    ConstraintWeights_5:float=0.0221
   ) -> (Pose) {
    alphaBoolBlend {blendOption=Linear}
    boneToModify {boneName=CarapaceWings_r_base}
    constraintSetup [
     {targetBone {boneName=spine_03}}
     {targetBone {boneName=clavicle_r}}
     {targetBone {boneName=spine_05}}
     {targetBone {boneName=spine_04}}
     {targetBone {boneName=upperarm_r}}
     {targetBone {boneName=deform_upperarm_bk_r1}}
    ]
    constraintWeights [1,1,1,1,1,1]
   }
   n29 = "Constraint - CarapaceWings_l_A (PPPPPPP)"!(
    ComponentPose=@n28
    Alpha=1
    ConstraintWeights_0=0.2875
    ConstraintWeights_1:float=0.1932
    ConstraintWeights_2:float=0.2772
    ConstraintWeights_3:float=0.1203
    ConstraintWeights_4:float=0.0157
    ConstraintWeights_5:float=0.0685
    ConstraintWeights_6:float=0.0376
   ) -> (Pose) {
    alphaBoolBlend {blendOption=Linear}
    boneToModify {boneName=CarapaceWings_l_A}
    constraintSetup [
     {targetBone {boneName=deform_upperarm_bk_l1}}
     {targetBone {boneName=upperarm_l}}
     {targetBone {boneName=clavicle_l}}
     {targetBone {boneName=spine_05}}
     {targetBone {boneName=spine_03}}
     {targetBone {boneName=deform_upperarm_up_l1}}
     {targetBone {boneName=spine_04}}
    ]
    constraintWeights [1,1,1,1,1,1,1]
   }
   n30 = "Constraint - CarapaceWings_r_A (PPPPPPP)"!(
    ComponentPose=@n29
    Alpha=1
    ConstraintWeights_0=0.2875
    ConstraintWeights_1:float=0.1932
    ConstraintWeights_2:float=0.2772
    ConstraintWeights_3:float=0.1203
    ConstraintWeights_4:float=0.0157
    ConstraintWeights_5:float=0.0685
    ConstraintWeights_6:float=0.0376
   ) -> (Pose) {
    alphaBoolBlend {blendOption=Linear}
    boneToModify {boneName=CarapaceWings_r_A}
    constraintSetup [
     {targetBone {boneName=deform_upperarm_bk_r1}}
     {targetBone {boneName=upperarm_r}}
     {targetBone {boneName=clavicle_r}}
     {targetBone {boneName=spine_05}}
     {targetBone {boneName=spine_03}}
     {targetBone {boneName=deform_upperarm_up_r1}}
     {targetBone {boneName=spine_04}}
    ]
    constraintWeights [1,1,1,1,1,1,1]
   }
   n31 = "Constraint - CarapaceWings_l_B (PPPP)"!(
    ComponentPose=@n30
    Alpha=1
    ConstraintWeights_0=0.3917
    ConstraintWeights_1:float=0.2633
    ConstraintWeights_2:float=0.2303
    ConstraintWeights_3:float=0.1147
   ) -> (Pose) {
    alphaBoolBlend {blendOption=Linear}
    boneToModify {boneName=CarapaceWings_l_B}
    constraintSetup [{targetBone {boneName=spine_03}}, {targetBone {boneName=spine_04}}, {targetBone {boneName=spine_05}}, {targetBone {boneName=clavicle_l}}]
    constraintWeights [1,1,1,1]
   }
   n32 = "Constraint - CarapaceWings_r_B (PPPP)"!(
    ComponentPose=@n31
    Alpha=1
    ConstraintWeights_0=0.3917
    ConstraintWeights_1:float=0.2633
    ConstraintWeights_2:float=0.2303
    ConstraintWeights_3:float=0.1147
   ) -> (Pose) {
    alphaBoolBlend {blendOption=Linear}
    boneToModify {boneName=CarapaceWings_r_B}
    constraintSetup [{targetBone {boneName=spine_03}}, {targetBone {boneName=spine_04}}, {targetBone {boneName=spine_05}}, {targetBone {boneName=clavicle_r}}]
    constraintWeights [1,1,1,1]
   }
   n33 = Control Rig(Source=Component To Local(ComponentPose=@n32)) {
    alphaBoolBlend {blendOption=Linear}
    controlRigAssetReference {blueprintRigClass=&/M_MED_MoonShock_Wings_ControlRig.M_MED_MoonShock_Wings_ControlRig_C}
   }
   n34 = "Constraint - upperarmDn_l (PP)"!(ComponentPose=Local To Component(LocalPose=@n33), Alpha=1, ConstraintWeights_0=0.3, ConstraintWeights_1:float=0.7) -> (Pose) {
    alphaBoolBlend {blendOption=Linear}
    boneToModify {boneName=upperarmDn_l}
    constraintSetup [{targetBone {boneName=spine_05}}, {targetBone {boneName=clavicle_l}}]
    constraintWeights [1,1]
   }
   n35 = "Constraint - upperarmDn_r (PP)"!(ComponentPose=@n34, Alpha=1, ConstraintWeights_0=0.3, ConstraintWeights_1:float=0.7) -> (Pose) {
    alphaBoolBlend {blendOption=Linear}
    boneToModify {boneName=upperarmDn_r}
    constraintSetup [{targetBone {boneName=spine_05}}, {targetBone {boneName=clavicle_r}}]
    constraintWeights [1,1]
   }
   n36 = "Transform (Modify) Bone - Bone: upperarmDn_l"(ComponentPose=@n35, Alpha=Get Shoulder_rot_dn_l() -> (Shoulder_rot_dn_l:double), Translation="-5,0,5", Rotation="0,0,0", Scale="1,1,1") {
    rotationMode=BMM_Additive
    rotationSpace=BCS_BoneSpace
    scaleMode=BMM_Additive
    scaleSpace=BCS_BoneSpace
    translationMode=BMM_Additive
    translationSpace=BCS_BoneSpace
    alphaBoolBlend {blendOption=Linear}
    boneToModify {boneName=upperarmDn_l}
    translation {x=-5, z=5}
   }
   n37 = "Transform (Modify) Bone - Bone: upperarmDn_r"(ComponentPose=@n36, Alpha=Get Shoulder_rot_dn_r() -> (Shoulder_rot_dn_r:double), Translation="5,0,-5", Rotation="0,0,0", Scale="1,1,1") {
    rotationMode=BMM_Additive
    rotationSpace=BCS_BoneSpace
    scaleMode=BMM_Additive
    scaleSpace=BCS_BoneSpace
    translationMode=BMM_Additive
    translationSpace=BCS_BoneSpace
    alphaBoolBlend {blendOption=Linear}
    boneToModify {boneName=upperarmDn_r}
    translation {x=5, z=-5}
   }
   n38 = "Transform (Modify) Bone - Bone: upperarmDn_l"(ComponentPose=@n37, Alpha=Get Shoulder_rot_up_l() -> (Shoulder_rot_up_l:double), Translation="-5,0,16", Rotation="0,0,0", Scale="1,1,1") {
    rotationMode=BMM_Additive
    rotationSpace=BCS_BoneSpace
    scaleMode=BMM_Additive
    scaleSpace=BCS_BoneSpace
    translationMode=BMM_Additive
    translationSpace=BCS_BoneSpace
    alphaBoolBlend {blendOption=Linear}
    boneToModify {boneName=upperarmDn_l}
    translation {x=-5, z=16}
   }
   n39 = "Transform (Modify) Bone - Bone: upperarmDn_r"(ComponentPose=@n38, Alpha=Get Shoulder_rot_up_r() -> (Shoulder_rot_up_r:double), Translation="5,0,-16", Rotation="0,0,0", Scale="1,1,1") {
    rotationMode=BMM_Additive
    rotationSpace=BCS_BoneSpace
    scaleMode=BMM_Additive
    scaleSpace=BCS_BoneSpace
    translationMode=BMM_Additive
    translationSpace=BCS_BoneSpace
    alphaBoolBlend {blendOption=Linear}
    boneToModify {boneName=upperarmDn_r}
    translation {x=5, z=-16}
   }
   Save cached pose 'BodyPose'(Pose=Component To Local(ComponentPose=@n39)) {cacheName=BodyPose}
  ]}
 wingsPosition {
  entry=Base
  path=AnimGraph.AnimGraphNode_StateMachine_0.wingsPosition
  ids=id
  nodes [
   AnimStateNode_0 = Base() {
    is=title AnimStateNode
    stateEntered {guid=#1}
    stateFullyBlended {guid=#2}
    stateLeft {guid=#3}
   }
   AnimStateNode_1 = Jumping() {
    is=title AnimStateNode
    stateEntered {guid=#4}
    stateFullyBlended {guid=#5}
    stateLeft {guid=#6}
   }
   AnimStateNode_2 = Falling() {
    is=title AnimStateNode
    stateEntered {guid=#7}
    stateFullyBlended {guid=#8}
    stateLeft {guid=#9}
   }
   AnimStateNode_3 = Targeting() {
    is=title AnimStateNode
    stateEntered {guid=#10}
    stateFullyBlended {guid=#11}
    stateLeft {guid=#12}
   }
   AnimStateNode_4 = Crouching() {
    is=title AnimStateNode
    stateEntered {guid=#13}
    stateFullyBlended {guid=#14}
    stateLeft {guid=#15}
   }
   AnimStateNode_5 = SkydiveMode() {
    is=title AnimStateNode
    stateEntered {guid=#16}
    stateFullyBlended {guid=#17}
    stateLeft {guid=#18}
   }
   AnimStateNode_9 = Sliding() {
    is=title AnimStateNode
    stateEntered {guid=#19}
    stateFullyBlended {guid=#20}
    stateLeft {guid=#21}
   }
   AnimStateTransitionNode_0 = AnimStateTransitionNode() {
    blendMode=Custom
    customBlendCurve=/Game/Accessories/FORT_Backpacks/M_MED_Demon_Pack/Animation/JumpTransitionBlendCurve
    priorityOrder=2
    transitionEnd {guid=#22}
    transitionInterrupt {guid=#23}
    transitionStart {guid=#24}
   }
   AnimStateTransitionNode_1 = AnimStateTransitionNode() {
    blendMode=Custom
    crossfadeDuration=0.5
    customBlendCurve=/Game/Accessories/FORT_Backpacks/M_MED_Demon_Pack/Animation/JumpTransitionBlendOutCurve
    transitionEnd {guid=#25}
    transitionInterrupt {guid=#26}
    transitionStart {guid=#27}
   }
   AnimStateTransitionNode_11 = AnimStateTransitionNode() {
    transitionEnd {guid=#28}
    transitionInterrupt {guid=#29}
    transitionStart {guid=#30}
   }
   AnimStateTransitionNode_12 = AnimStateTransitionNode() {
    transitionEnd {guid=#31}
    transitionInterrupt {guid=#32}
    transitionStart {guid=#33}
   }
   AnimStateTransitionNode_13 = AnimStateTransitionNode() {
    transitionEnd {guid=#34}
    transitionInterrupt {guid=#35}
    transitionStart {guid=#36}
   }
   AnimStateTransitionNode_15 = AnimStateTransitionNode() {
    transitionEnd {guid=#37}
    transitionInterrupt {guid=#38}
    transitionStart {guid=#39}
   }
   AnimStateTransitionNode_2 = AnimStateTransitionNode() {
    transitionEnd {guid=#40}
    transitionInterrupt {guid=#41}
    transitionStart {guid=#42}
   }
   AnimStateTransitionNode_3 = AnimStateTransitionNode() {
    blendMode=Custom
    customBlendCurve=&JumpTransitionBlendOutCurve
    transitionEnd {guid=#43}
    transitionInterrupt {guid=#44}
    transitionStart {guid=#45}
   }
   AnimStateTransitionNode_4 = AnimStateTransitionNode() {
    blendMode=Custom
    customBlendCurve=&JumpTransitionBlendCurve
    transitionEnd {guid=#46}
    transitionInterrupt {guid=#47}
    transitionStart {guid=#48}
   }
   AnimStateTransitionNode_5 = AnimStateTransitionNode() {
    transitionEnd {guid=#49}
    transitionInterrupt {guid=#50}
    transitionStart {guid=#51}
   }
   AnimStateTransitionNode_6 = AnimStateTransitionNode() {
    crossfadeDuration=0.25
    transitionEnd {guid=#52}
    transitionInterrupt {guid=#53}
    transitionStart {guid=#54}
   }
   AnimStateTransitionNode_7 = AnimStateTransitionNode() {
    crossfadeDuration=0.25
    transitionEnd {guid=#55}
    transitionInterrupt {guid=#56}
    transitionStart {guid=#57}
   }
   AnimStateTransitionNode_8 = AnimStateTransitionNode()
   AnimStateTransitionNode_9 = AnimStateTransitionNode()
  ]
  links [
   {
    dst=Crouching
    rule=AnimGraph.AnimGraphNode_StateMachine_0.wingsPosition.AnimStateTransitionNode_12.AnimationTransitionGraph_0
    src=Base
    via=AnimStateTransitionNode_12
   }
   {
    dst=Falling
    rule=AnimGraph.AnimGraphNode_StateMachine_0.wingsPosition.AnimStateTransitionNode_4.AnimationTransitionGraph_0
    src=Base
    via=AnimStateTransitionNode_4
   }
   {
    dst=Jumping
    rule=AnimGraph.AnimGraphNode_StateMachine_0.wingsPosition.AnimStateTransitionNode_0.AnimationTransitionGraph_0
    src=Base
    via=AnimStateTransitionNode_0
   }
   {
    dst=SkydiveMode
    rule=AnimGraph.AnimGraphNode_StateMachine_0.wingsPosition.AnimStateTransitionNode_15.AnimationTransitionGraph_0
    src=Base
    via=AnimStateTransitionNode_15
   }
   {
    dst=Targeting
    rule=AnimGraph.AnimGraphNode_StateMachine_0.wingsPosition.AnimStateTransitionNode_6.AnimationTransitionGraph_0
    src=Base
    via=AnimStateTransitionNode_6
   }
   {
    dst=Base
    rule=AnimGraph.AnimGraphNode_StateMachine_0.wingsPosition.AnimStateTransitionNode_13.AnimationTransitionGraph_0
    src=Crouching
    via=AnimStateTransitionNode_13
   }
   {
    dst=Sliding
    rule=AnimGraph.AnimGraphNode_StateMachine_0.wingsPosition.AnimStateTransitionNode_9.Transition
    src=Crouching
    via=AnimStateTransitionNode_9
   }
   {
    dst=Base
    rule=AnimGraph.AnimGraphNode_StateMachine_0.wingsPosition.AnimStateTransitionNode_3.AnimationTransitionGraph_0
    src=Falling
    via=AnimStateTransitionNode_3
   }
   {
    dst=Base
    rule=AnimGraph.AnimGraphNode_StateMachine_0.wingsPosition.AnimStateTransitionNode_1.AnimationTransitionGraph_0
    src=Jumping
    via=AnimStateTransitionNode_1
   }
   {
    dst=Falling
    rule=AnimGraph.AnimGraphNode_StateMachine_0.wingsPosition.AnimStateTransitionNode_2.AnimationTransitionGraph_0
    src=Jumping
    via=AnimStateTransitionNode_2
   }
   {
    dst=SkydiveMode
    rule=AnimGraph.AnimGraphNode_StateMachine_0.wingsPosition.AnimStateTransitionNode_5.AnimationTransitionGraph_0
    src=Jumping
    via=AnimStateTransitionNode_5
   }
   {
    dst=Base
    rule=AnimGraph.AnimGraphNode_StateMachine_0.wingsPosition.AnimStateTransitionNode_11.AnimationTransitionGraph_0
    src=SkydiveMode
    via=AnimStateTransitionNode_11
   }
   {
    dst=Crouching
    rule=AnimGraph.AnimGraphNode_StateMachine_0.wingsPosition.AnimStateTransitionNode_8.Transition
    src=Sliding
    via=AnimStateTransitionNode_8
   }
   {
    dst=Base
    rule=AnimGraph.AnimGraphNode_StateMachine_0.wingsPosition.AnimStateTransitionNode_7.AnimationTransitionGraph_0
    src=Targeting
    via=AnimStateTransitionNode_7
   }
  ]
 }
 Base {
  path=AnimGraph.AnimGraphNode_StateMachine_0.wingsPosition.AnimStateNode_0.Base
  nodes [Output Animation Pose(Result=Blendspace Player 'MoonShock_Wings_Leans_Ground'(X=Get PlayerVector_LR() -> (PlayerVector_LR:double), Y=Get PlayerVector_FwdBwd() -> (PlayerVector_FwdBwd:double)) {blendSpace=/BRCosmetics/Characters/Player/Male/Medium/Bodies/M_MED_MoonShock_Wings/Animarion/MoonShock_Wings_Leans_Ground})]
 }
 Jumping {
  path=AnimGraph.AnimGraphNode_StateMachine_0.wingsPosition.AnimStateNode_1.Jumping
  nodes [
   n1 = MoonShock_Wings_PoseAsset : Jump!(PoseWeight:float=1) -> (Pose:Pose Link Structure) {
    is=title AnimGraphNode_PoseByName
    poseAsset=/BRCosmetics/Characters/Player/Male/Medium/Bodies/M_MED_MoonShock_Wings/Animarion/MoonShock_Wings_PoseAsset
    poseName=Jump
   }
   n2 = "Transform (Modify) Bone - Bone: CarapaceWings_l"(ComponentPose=Local To Component(LocalPose=@n1), Alpha=1, Translation="0,0,0", Rotation="0,0,-10", Scale="1,1,1") {
    rotationMode=BMM_Additive
    rotationSpace=BCS_BoneSpace
    scaleMode=BMM_Additive
    scaleSpace=BCS_BoneSpace
    translationMode=BMM_Additive
    translationSpace=BCS_BoneSpace
    alphaBoolBlend {blendOption=Linear}
    boneToModify {boneName=CarapaceWings_l}
    rotation {roll=-10}
   }
   n3 = "Transform (Modify) Bone - Bone: CarapaceWings_r"(ComponentPose=@n2, Alpha=1, Translation="0,0,0", Rotation="0,0,-10", Scale="1,1,1") {
    rotationMode=BMM_Additive
    rotationSpace=BCS_BoneSpace
    scaleMode=BMM_Additive
    scaleSpace=BCS_BoneSpace
    translationMode=BMM_Additive
    translationSpace=BCS_BoneSpace
    alphaBoolBlend {blendOption=Linear}
    boneToModify {boneName=CarapaceWings_r}
    rotation {roll=-10}
   }
   Output Animation Pose(Result=Component To Local(ComponentPose=@n3))
  ]
 }
 Falling {
  path=AnimGraph.AnimGraphNode_StateMachine_0.wingsPosition.AnimStateNode_2.Falling
  nodes [
   n1 = MoonShock_Wings_PoseAsset : Jump(PoseWeight=1) {poseAsset=&MoonShock_Wings_PoseAsset, poseName=Jump}
   n2 = "Transform (Modify) Bone - Bone: CarapaceWings_l"(ComponentPose=Local To Component(LocalPose=@n1), Alpha=1, Translation="0,0,0", Rotation="0,0,-10", Scale="1,1,1") {
    rotationMode=BMM_Additive
    rotationSpace=BCS_BoneSpace
    scaleMode=BMM_Additive
    scaleSpace=BCS_BoneSpace
    translationMode=BMM_Additive
    translationSpace=BCS_BoneSpace
    alphaBoolBlend {blendOption=Linear}
    boneToModify {boneName=CarapaceWings_l}
    rotation {roll=-10}
   }
   n3 = "Transform (Modify) Bone - Bone: CarapaceWings_r"(ComponentPose=@n2, Alpha=1, Translation="0,0,0", Rotation="0,0,-10", Scale="1,1,1") {
    rotationMode=BMM_Additive
    rotationSpace=BCS_BoneSpace
    scaleMode=BMM_Additive
    scaleSpace=BCS_BoneSpace
    translationMode=BMM_Additive
    translationSpace=BCS_BoneSpace
    alphaBoolBlend {blendOption=Linear}
    boneToModify {boneName=CarapaceWings_r}
    rotation {roll=-10}
   }
   Output Animation Pose(Result=Component To Local(ComponentPose=@n3))
  ]
 }
 Targeting {
  path=AnimGraph.AnimGraphNode_StateMachine_0.wingsPosition.AnimStateNode_3.Targeting
  nodes [
   n1 = MoonShock_Wings_PoseAsset : Targeting!(PoseWeight:float=1) -> (Pose:Pose Link Structure) {
    is=title AnimGraphNode_PoseByName
    poseAsset=&MoonShock_Wings_PoseAsset
    poseName=Targeting
   }
   Output Animation Pose(Result=@n1)
  ]
 }
 Crouching {
  path=AnimGraph.AnimGraphNode_StateMachine_0.wingsPosition.AnimStateNode_4.Crouching
  nodes [
   MoonShock_Wings_PoseAsset : Crouch!(PoseWeight:float=1) -> (Pose:Pose Link Structure) {
    is=title AnimGraphNode_PoseByName
    poseAsset=&MoonShock_Wings_PoseAsset
    poseName=Crouch
   }
   AnimGraphNode_StateMachine_0 = Crouch() {is=title AnimGraphNode_StateMachine}
   Output Animation Pose(Result=@AnimGraphNode_StateMachine_0)
  ]
 }
 Crouch {
  entry=Blank
  path=AnimGraph.AnimGraphNode_StateMachine_0.wingsPosition.AnimStateNode_4.Crouching.AnimGraphNode_StateMachine_0.Crouch
  ids=id
  nodes [
   AnimStateNode_0 = Blank() {
    is=title AnimStateNode
    stateEntered {guid=#58}
    stateFullyBlended {guid=#59}
    stateLeft {guid=#60}
   }
   AnimStateNode_2 = CrouchSprinting() {
    is=title AnimStateNode
    stateEntered {guid=#61}
    stateFullyBlended {guid=#62}
    stateLeft {guid=#63}
   }
   AnimStateNode_27 = CrouchWalk() {
    is=title AnimStateNode
    stateEntered {guid=#64}
    stateFullyBlended {guid=#65}
    stateLeft {guid=#66}
   }
   AnimStateNode_29 = Crouch() {
    is=title AnimStateNode
    stateEntered {guid=#67}
    stateFullyBlended {guid=#68}
    stateLeft {guid=#69}
   }
   AnimStateTransitionNode_11 = AnimStateTransitionNode() {
    transitionEnd {guid=#70}
    transitionInterrupt {guid=#71}
    transitionStart {guid=#72}
   }
   AnimStateTransitionNode_12 = AnimStateTransitionNode() {
    transitionEnd {guid=#73}
    transitionInterrupt {guid=#74}
    transitionStart {guid=#75}
   }
   AnimStateTransitionNode_17 = AnimStateTransitionNode()
   AnimStateTransitionNode_4 = AnimStateTransitionNode() {
    transitionEnd {guid=#76}
    transitionInterrupt {guid=#77}
    transitionStart {guid=#78}
   }
   AnimStateTransitionNode_47 = AnimStateTransitionNode() {
    transitionEnd {guid=#79}
    transitionInterrupt {guid=#80}
    transitionStart {guid=#81}
   }
   AnimStateTransitionNode_55 = AnimStateTransitionNode()
  ]
  links [
   {
    dst=Crouch
    rule=AnimGraph.AnimGraphNode_StateMachine_0.wingsPosition.AnimStateNode_4.Crouching.AnimGraphNode_StateMachine_0.Crouch.AnimStateTransitionNode_11.AnimationTransitionGraph_0
    src=Blank
    via=AnimStateTransitionNode_11
   }
   {
    dst=CrouchWalk
    rule=AnimGraph.AnimGraphNode_StateMachine_0.wingsPosition.AnimStateNode_4.Crouching.AnimGraphNode_StateMachine_0.Crouch.AnimStateTransitionNode_12.Transition
    src=Blank
    via=AnimStateTransitionNode_12
   }
   {
    dst=CrouchWalk
    rule=AnimGraph.AnimGraphNode_StateMachine_0.wingsPosition.AnimStateNode_4.Crouching.AnimGraphNode_StateMachine_0.Crouch.AnimStateTransitionNode_17.Transition
    src=Crouch
    via=AnimStateTransitionNode_17
   }
   {
    dst=CrouchWalk
    rule=AnimGraph.AnimGraphNode_StateMachine_0.wingsPosition.AnimStateNode_4.Crouching.AnimGraphNode_StateMachine_0.Crouch.AnimStateTransitionNode_55.AnimationTransitionGraph_1
    src=CrouchSprinting
    via=AnimStateTransitionNode_55
   }
   {
    dst=Crouch
    rule=AnimGraph.AnimGraphNode_StateMachine_0.wingsPosition.AnimStateNode_4.Crouching.AnimGraphNode_StateMachine_0.Crouch.AnimStateTransitionNode_4.Transition
    src=CrouchWalk
    via=AnimStateTransitionNode_4
   }
   {
    dst=CrouchSprinting
    rule=AnimGraph.AnimGraphNode_StateMachine_0.wingsPosition.AnimStateNode_4.Crouching.AnimGraphNode_StateMachine_0.Crouch.AnimStateTransitionNode_47.AnimationTransitionGraph_1
    src=CrouchWalk
    via=AnimStateTransitionNode_47
   }
  ]
 }
 Blank {
  path=AnimGraph.AnimGraphNode_StateMachine_0.wingsPosition.AnimStateNode_4.Crouching.AnimGraphNode_StateMachine_0.Crouch.AnimStateNode_0.Blank
  nodes [Output Animation Pose()]
 }
 CrouchSprinting {
  path=AnimGraph.AnimGraphNode_StateMachine_0.wingsPosition.AnimStateNode_4.Crouching.AnimGraphNode_StateMachine_0.Crouch.AnimStateNode_2.CrouchSprinting
  nodes [
   n1 = MoonShock_Wings_PoseAsset : Crouch(PoseWeight=1) {poseAsset=&MoonShock_Wings_PoseAsset, poseName=Crouch}
   Output Animation Pose(Result=@n1)
  ]
 }
 CrouchWalk {
  path=AnimGraph.AnimGraphNode_StateMachine_0.wingsPosition.AnimStateNode_4.Crouching.AnimGraphNode_StateMachine_0.Crouch.AnimStateNode_27.CrouchWalk
  nodes [
   n1 = MoonShock_Wings_PoseAsset : Crouch(PoseWeight=1) {poseAsset=&MoonShock_Wings_PoseAsset, poseName=Crouch}
   Output Animation Pose(Result=@n1)
  ]
 }
 Crouch {
  path=AnimGraph.AnimGraphNode_StateMachine_0.wingsPosition.AnimStateNode_4.Crouching.AnimGraphNode_StateMachine_0.Crouch.AnimStateNode_29.Crouch
  nodes [
   n1 = MoonShock_Wings_PoseAsset : Crouch(PoseWeight=1) {poseAsset=&MoonShock_Wings_PoseAsset, poseName=Crouch}
   n2 = "Transform (Modify) Bone - Bone: CarapaceWings_l"(ComponentPose=Local To Component(LocalPose=@n1), Alpha=1, Translation="0,0,0", Rotation="-5,0,10", Scale="1,1,1") {
    rotationMode=BMM_Additive
    rotationSpace=BCS_BoneSpace
    scaleMode=BMM_Additive
    scaleSpace=BCS_BoneSpace
    translationMode=BMM_Additive
    translationSpace=BCS_BoneSpace
    alphaBoolBlend {blendOption=Linear}
    boneToModify {boneName=CarapaceWings_l}
    rotation {pitch=-5, roll=10}
   }
   n3 = "Transform (Modify) Bone - Bone: CarapaceWings_r"(ComponentPose=@n2, Alpha=1, Translation="0,0,0", Rotation="-5,0,10", Scale="1,1,1") {
    rotationMode=BMM_Additive
    rotationSpace=BCS_BoneSpace
    scaleMode=BMM_Additive
    scaleSpace=BCS_BoneSpace
    translationMode=BMM_Additive
    translationSpace=BCS_BoneSpace
    alphaBoolBlend {blendOption=Linear}
    boneToModify {boneName=CarapaceWings_r}
    rotation {pitch=-5, roll=10}
   }
   Output Animation Pose(Result=Component To Local(ComponentPose=@n3))
  ]
 }
 AnimationTransitionGraph_0 {
  path=AnimGraph.AnimGraphNode_StateMachine_0.wingsPosition.AnimStateNode_4.Crouching.AnimGraphNode_StateMachine_0.Crouch.AnimStateTransitionNode_11.AnimationTransitionGraph_0
  nodes [Result(bCanEnterTransition=NOT Boolean(A=Get bIsMoving2D()))]
 }
 Transition {
  path=AnimGraph.AnimGraphNode_StateMachine_0.wingsPosition.AnimStateNode_4.Crouching.AnimGraphNode_StateMachine_0.Crouch.AnimStateTransitionNode_12.Transition
  nodes [Result(bCanEnterTransition=Get bIsMoving2D())]
 }
 Transition {
  path=AnimGraph.AnimGraphNode_StateMachine_0.wingsPosition.AnimStateNode_4.Crouching.AnimGraphNode_StateMachine_0.Crouch.AnimStateTransitionNode_17.Transition
  nodes [Result(bCanEnterTransition=Get bIsMoving2D())]
 }
 Transition {
  path=AnimGraph.AnimGraphNode_StateMachine_0.wingsPosition.AnimStateNode_4.Crouching.AnimGraphNode_StateMachine_0.Crouch.AnimStateTransitionNode_4.Transition
  nodes [Result(bCanEnterTransition=NOT Boolean(A=Get bIsMoving2D()))]
 }
 AnimationTransitionGraph_1 {
  path=AnimGraph.AnimGraphNode_StateMachine_0.wingsPosition.AnimStateNode_4.Crouching.AnimGraphNode_StateMachine_0.Crouch.AnimStateTransitionNode_47.AnimationTransitionGraph_1
  nodes [Result(bCanEnterTransition=Get bIsCrouchSprinting())]
 }
 AnimationTransitionGraph_1 {
  path=AnimGraph.AnimGraphNode_StateMachine_0.wingsPosition.AnimStateNode_4.Crouching.AnimGraphNode_StateMachine_0.Crouch.AnimStateTransitionNode_55.AnimationTransitionGraph_1
  nodes [Result(bCanEnterTransition=NOT Boolean(A=Get bIsCrouchSprinting()))]
 }
 SkydiveMode {
  path=AnimGraph.AnimGraphNode_StateMachine_0.wingsPosition.AnimStateNode_5.SkydiveMode
  nodes [
   AnimGraphNode_StateMachine_0 = Skydiving State() {is=title AnimGraphNode_StateMachine}
   n1 = "Transform (Modify) Bone - Bone: CarapaceWings_l"(ComponentPose=Local To Component(LocalPose=@AnimGraphNode_StateMachine_0), Alpha=1, Translation="0,0,0", Rotation="4,0,-10", Scale="1,1,1") {
    rotationMode=BMM_Additive
    rotationSpace=BCS_BoneSpace
    scaleMode=BMM_Additive
    scaleSpace=BCS_BoneSpace
    translationMode=BMM_Additive
    translationSpace=BCS_BoneSpace
    alphaBoolBlend {blendOption=Linear}
    boneToModify {boneName=CarapaceWings_l}
    rotation {pitch=4, roll=-10}
   }
   n2 = "Transform (Modify) Bone - Bone: CarapaceWings_r"(ComponentPose=@n1, Alpha=1, Translation="0,0,0", Rotation="4,0,-10", Scale="1,1,1") {
    rotationMode=BMM_Additive
    rotationSpace=BCS_BoneSpace
    scaleMode=BMM_Additive
    scaleSpace=BCS_BoneSpace
    translationMode=BMM_Additive
    translationSpace=BCS_BoneSpace
    alphaBoolBlend {blendOption=Linear}
    boneToModify {boneName=CarapaceWings_r}
    rotation {pitch=4, roll=-10}
   }
   Output Animation Pose(Result=Component To Local(ComponentPose=@n2))
  ]
 }
 Skydiving State {
  entry=Parachuting
  path=AnimGraph.AnimGraphNode_StateMachine_0.wingsPosition.AnimStateNode_5.SkydiveMode.AnimGraphNode_StateMachine_0.Skydiving State
  ids=id
  nodes [
   AnimStateNode_1 = Diving() {is=title AnimStateNode}
   AnimStateNode_2 = Floating() {is=title AnimStateNode}
   AnimStateNode_4 = Parachuting() {is=title AnimStateNode}
   AnimStateTransitionNode_1 = AnimStateTransitionNode() {crossfadeDuration=0.5}
   AnimStateTransitionNode_10 = AnimStateTransitionNode()
   AnimStateTransitionNode_12 = AnimStateTransitionNode()
   AnimStateTransitionNode_13 = AnimStateTransitionNode()
   AnimStateTransitionNode_3 = AnimStateTransitionNode()
   AnimStateTransitionNode_5 = AnimStateTransitionNode() {crossfadeDuration=0.5}
  ]
  links [
   {
    dst=Floating
    rule=AnimGraph.AnimGraphNode_StateMachine_0.wingsPosition.AnimStateNode_5.SkydiveMode.AnimGraphNode_StateMachine_0.Skydiving State.AnimStateTransitionNode_1.AnimationTransitionGraph_0
    src=Diving
    via=AnimStateTransitionNode_1
   }
   {
    dst=Parachuting
    rule=AnimGraph.AnimGraphNode_StateMachine_0.wingsPosition.AnimStateNode_5.SkydiveMode.AnimGraphNode_StateMachine_0.Skydiving State.AnimStateTransitionNode_10.AnimationTransitionGraph_1
    src=Diving
    via=AnimStateTransitionNode_10
   }
   {
    dst=Diving
    rule=AnimGraph.AnimGraphNode_StateMachine_0.wingsPosition.AnimStateNode_5.SkydiveMode.AnimGraphNode_StateMachine_0.Skydiving State.AnimStateTransitionNode_5.AnimationTransitionGraph_0
    src=Floating
    via=AnimStateTransitionNode_5
   }
   {
    dst=Parachuting
    rule=AnimGraph.AnimGraphNode_StateMachine_0.wingsPosition.AnimStateNode_5.SkydiveMode.AnimGraphNode_StateMachine_0.Skydiving State.AnimStateTransitionNode_3.AnimationTransitionGraph_1
    src=Floating
    via=AnimStateTransitionNode_3
   }
   {
    dst=Diving
    rule=AnimGraph.AnimGraphNode_StateMachine_0.wingsPosition.AnimStateNode_5.SkydiveMode.AnimGraphNode_StateMachine_0.Skydiving State.AnimStateTransitionNode_13.AnimationTransitionGraph_1
    src=Parachuting
    via=AnimStateTransitionNode_13
   }
   {
    dst=Floating
    rule=AnimGraph.AnimGraphNode_StateMachine_0.wingsPosition.AnimStateNode_5.SkydiveMode.AnimGraphNode_StateMachine_0.Skydiving State.AnimStateTransitionNode_12.AnimationTransitionGraph_1
    src=Parachuting
    via=AnimStateTransitionNode_12
   }
  ]
 }
 Diving {
  path=AnimGraph.AnimGraphNode_StateMachine_0.wingsPosition.AnimStateNode_5.SkydiveMode.AnimGraphNode_StateMachine_0.Skydiving State.AnimStateNode_1.Diving
  nodes [
   n1 = MoonShock_Wings_PoseAsset : Dive!(PoseWeight:float=1) -> (Pose:Pose Link Structure) {
    is=title AnimGraphNode_PoseByName
    poseAsset=&MoonShock_Wings_PoseAsset
    poseName=Dive
   }
   n2 = "Transform (Modify) Bone - Bone: attach_backpack"(ComponentPose=Local To Component(LocalPose=@n1), Alpha=1, Translation="0,0,0", Rotation="0,0,10", Scale="1,1,1") {
    rotationMode=BMM_Additive
    translationMode=BMM_Additive
    alphaBoolBlend {blendOption=Linear}
    boneToModify {boneName=attach_backpack}
    rotation {roll=10}
   }
   Output Animation Pose(Result=Component To Local(ComponentPose=@n2))
  ]
 }
 Floating {
  path=AnimGraph.AnimGraphNode_StateMachine_0.wingsPosition.AnimStateNode_5.SkydiveMode.AnimGraphNode_StateMachine_0.Skydiving State.AnimStateNode_2.Floating
  nodes [
   n1 = MoonShock_Wings_PoseAsset : Float!(PoseWeight:float=1) -> (Pose:Pose Link Structure) {
    is=title AnimGraphNode_PoseByName
    poseAsset=&MoonShock_Wings_PoseAsset
    poseName=Float
   }
   Output Animation Pose(Result=@n1)
  ]
 }
 Parachuting {
  path=AnimGraph.AnimGraphNode_StateMachine_0.wingsPosition.AnimStateNode_5.SkydiveMode.AnimGraphNode_StateMachine_0.Skydiving State.AnimStateNode_4.Parachuting
  nodes [
   n1 = MoonShock_Wings_PoseAsset : Float(PoseWeight=0.8) {poseAsset=&MoonShock_Wings_PoseAsset, poseName=Float}
   Output Animation Pose(Result=@n1)
  ]
 }
 AnimationTransitionGraph_0 {
  path=AnimGraph.AnimGraphNode_StateMachine_0.wingsPosition.AnimStateNode_5.SkydiveMode.AnimGraphNode_StateMachine_0.Skydiving State.AnimStateTransitionNode_1.AnimationTransitionGraph_0
  nodes [Result(bCanEnterTransition=NOT Boolean(A=Get bIsSkydiveDiveMode()))]
 }
 AnimationTransitionGraph_1 {
  path=AnimGraph.AnimGraphNode_StateMachine_0.wingsPosition.AnimStateNode_5.SkydiveMode.AnimGraphNode_StateMachine_0.Skydiving State.AnimStateTransitionNode_10.AnimationTransitionGraph_1
  nodes [Result(bCanEnterTransition=Get bIsParachuteOpened())]
 }
 AnimationTransitionGraph_1 {
  path=AnimGraph.AnimGraphNode_StateMachine_0.wingsPosition.AnimStateNode_5.SkydiveMode.AnimGraphNode_StateMachine_0.Skydiving State.AnimStateTransitionNode_12.AnimationTransitionGraph_1
  nodes [Result(bCanEnterTransition=NOT Boolean(A=Get bIsParachuteOpened()))]
 }
 AnimationTransitionGraph_1 {
  path=AnimGraph.AnimGraphNode_StateMachine_0.wingsPosition.AnimStateNode_5.SkydiveMode.AnimGraphNode_StateMachine_0.Skydiving State.AnimStateTransitionNode_13.AnimationTransitionGraph_1
  nodes [Result(bCanEnterTransition=NOT Boolean(A=Get bIsParachuteOpened()))]
 }
 AnimationTransitionGraph_1 {
  path=AnimGraph.AnimGraphNode_StateMachine_0.wingsPosition.AnimStateNode_5.SkydiveMode.AnimGraphNode_StateMachine_0.Skydiving State.AnimStateTransitionNode_3.AnimationTransitionGraph_1
  nodes [Result(bCanEnterTransition=Get bIsParachuteOpened())]
 }
 AnimationTransitionGraph_0 {
  path=AnimGraph.AnimGraphNode_StateMachine_0.wingsPosition.AnimStateNode_5.SkydiveMode.AnimGraphNode_StateMachine_0.Skydiving State.AnimStateTransitionNode_5.AnimationTransitionGraph_0
  nodes [Result(bCanEnterTransition=Get bIsSkydiveDiveMode())]
 }
 Sliding {
  path=AnimGraph.AnimGraphNode_StateMachine_0.wingsPosition.AnimStateNode_9.Sliding
  nodes [
   n1 = MoonShock_Wings_PoseAsset : sprint!(PoseWeight:float=1) -> (Pose:Pose Link Structure) {
    is=title AnimGraphNode_PoseByName
    poseAsset=&MoonShock_Wings_PoseAsset
    poseName=sprint
   }
   Output Animation Pose(Result=@n1)
  ]
 }
 AnimationTransitionGraph_0 {
  path=AnimGraph.AnimGraphNode_StateMachine_0.wingsPosition.AnimStateTransitionNode_0.AnimationTransitionGraph_0
  nodes [Result(bCanEnterTransition=Get bIsJumping())]
 }
 AnimationTransitionGraph_0 {
  path=AnimGraph.AnimGraphNode_StateMachine_0.wingsPosition.AnimStateTransitionNode_1.AnimationTransitionGraph_0
  nodes [
   Result(bCanEnterTransition=NOT Boolean(A=Get bIsJumping()))
   Get bShouldPredictLanding() -> (bShouldPredictLanding:bool)
   Get bLandingPredicted() -> (bLandingPredicted:bool)
   Get bDebugLandPrediction() -> (bDebugLandPrediction:bool)
   Get bIsFalling()
  ]
 }
 AnimationTransitionGraph_0 {
  path=AnimGraph.AnimGraphNode_StateMachine_0.wingsPosition.AnimStateTransitionNode_11.AnimationTransitionGraph_0
  nodes [Result(bCanEnterTransition=NOT Boolean(A=Get bIsSkyDiving()))]
 }
 AnimationTransitionGraph_0 {
  path=AnimGraph.AnimGraphNode_StateMachine_0.wingsPosition.AnimStateTransitionNode_12.AnimationTransitionGraph_0
  nodes [Result(bCanEnterTransition=Get bIsCrouching())]
 }
 AnimationTransitionGraph_0 {
  path=AnimGraph.AnimGraphNode_StateMachine_0.wingsPosition.AnimStateTransitionNode_13.AnimationTransitionGraph_0
  nodes [Result(bCanEnterTransition=NOT Boolean(A=Get bIsCrouching()))]
 }
 AnimationTransitionGraph_0 {
  path=AnimGraph.AnimGraphNode_StateMachine_0.wingsPosition.AnimStateTransitionNode_15.AnimationTransitionGraph_0
  nodes [Result(bCanEnterTransition=Get bIsSkyDiving())]
 }
 AnimationTransitionGraph_0 {
  path=AnimGraph.AnimGraphNode_StateMachine_0.wingsPosition.AnimStateTransitionNode_2.AnimationTransitionGraph_0
  nodes [Result(bCanEnterTransition=Get bIsFalling())]
 }
 AnimationTransitionGraph_0 {
  path=AnimGraph.AnimGraphNode_StateMachine_0.wingsPosition.AnimStateTransitionNode_3.AnimationTransitionGraph_0
  nodes [Result(bCanEnterTransition=NOT Boolean(A=Get bIsFalling()))]
 }
 AnimationTransitionGraph_0 {
  path=AnimGraph.AnimGraphNode_StateMachine_0.wingsPosition.AnimStateTransitionNode_4.AnimationTransitionGraph_0
  nodes [Result(bCanEnterTransition=Get bIsFalling())]
 }
 AnimationTransitionGraph_0 {
  path=AnimGraph.AnimGraphNode_StateMachine_0.wingsPosition.AnimStateTransitionNode_5.AnimationTransitionGraph_0
  nodes [Result(bCanEnterTransition=Get bIsSkyDiving())]
 }
 AnimationTransitionGraph_0 {
  path=AnimGraph.AnimGraphNode_StateMachine_0.wingsPosition.AnimStateTransitionNode_6.AnimationTransitionGraph_0
  nodes [Result(bCanEnterTransition=Get bIsTargeting())]
 }
 AnimationTransitionGraph_0 {
  path=AnimGraph.AnimGraphNode_StateMachine_0.wingsPosition.AnimStateTransitionNode_7.AnimationTransitionGraph_0
  nodes [Result(bCanEnterTransition=NOT Boolean(A=Get bIsTargeting()))]
 }
 Transition {
  path=AnimGraph.AnimGraphNode_StateMachine_0.wingsPosition.AnimStateTransitionNode_8.Transition
  nodes [Result(bCanEnterTransition=Get isCrouching And Moving() -> (isCrouching And Moving:bool))]
 }
 Transition {
  path=AnimGraph.AnimGraphNode_StateMachine_0.wingsPosition.AnimStateTransitionNode_9.Transition
  nodes [Result(bCanEnterTransition=Get bIsSliding())]
 }
 EventGraph {nodes [
   Event BlueprintUpdateAnimation()
   > n1 = Sequence!(execute) -> (then_0, then_1, then_2:Exec, then_3:Exec)
   n2 = TryGetPawnOwner()
   > n3 = Cast To Character!(execute, Object=@n2) -> (then, CastFailed, AsCharacter:Character Object Reference) {is=title K2Node_DynamicCast}
   n4 = GetCurrentAcceleration!(self:Character Movement Component Object Reference=Get CharacterMovement!(self:Character Object Reference=@n3) -> (CharacterMovement:Character Movement Component Object Reference)) -> (ReturnValue:Vector)
   n5 = Get Actor Rotation!(self:Actor Object Reference=@n2) -> (ReturnValue:Rotator) {is=title K2Node_CallFunction}
   n6 = BreakVector(InVec=vector * vector(A:Vector=Get Rotation X Vector!(InRot:Rotator=@n5) -> (ReturnValue:Vector) {is=title K2Node_CallFunction}, B:Vector=@n4) -> (ReturnValue:Vector) {is=title K2Node_CommutativeAssociativeBinaryOperator})
   > Set PlayerVector_FwdBwd(PlayerVector_FwdBwd:double=float + float(A=@n6, B=@n6.Y) {is=title K2Node_CommutativeAssociativeBinaryOperator}) -> (Output_Get:double)
   n7 = BreakVector(InVec=vector * vector(A=GetAxes!(A:Rotator=@n5) -> (X:Vector, Y:Vector, Z:Vector).Y, B=@n4))
   > Set PlayerVector_LR(PlayerVector_LR:double=float + float(A=@n7, B=@n7.Y)) -> (Output_Get:double)
   Set isDivingAndWearingCape(execute=@n1.then_2, isDivingAndWearingCape:bool=AND Boolean(A=Get isWearingCape(), B=Get bIsSkyDiving())) -> (Output_Get:bool)
   n8 = NOT Boolean(A=Get bIsSliding())
   n9 = Get bIsMoving2D()
   n10 = Get bIsCrouching()
   Set isCrouching And Moving(
    execute=@n1.then_3
    isCrouching And Moving:bool=OR Boolean!(A=AND Boolean(A=@n10, B=@n8), B=AND Boolean(A=@n8, B=@n9), C:bool=AND Boolean!(A=@n10, B=@n8, C:bool=@n9) -> (ReturnValue)) -> (ReturnValue)
   ) -> (Output_Get:bool)
   n11 = CheckForBackbling()
   Set isWearingBackpack(execute=@n1.then_1, isWearingBackpack:bool=AND Boolean!(A=@n11, B=NOT Boolean(A=@n11.bIsBackblingLong)) -> (ReturnValue)) -> (Output_Get:bool)
   > Set isWearingBackpack(isWearingBackpack=NOT Boolean(A=@n11))
   > Set isWearingCape(isWearingCape=OR Boolean!(A=@n11.bIsBackblingACape, B=@n11.bIsBackblingLong) -> (ReturnValue))
   n12 = Event OnCharacterCustomizationLoaded()
   > n13 = Sequence()
   Set Character Mesh(execute=@n13.then_1, Character Mesh:Skeletal Mesh Component Object Reference=@n12.CustomBaseMesh) -> (Output_Get:Skeletal Mesh Component Object Reference)
   Set Body Mesh(
    execute=@n13.then_2
    Body Mesh:Skeletal Mesh Component Object Reference=Get Array Item with EFort Custom Part Type(Array=@n12.CustomMeshParts)
   ) -> (Output_Get:Skeletal Mesh Component Object Reference)
   n14 = Get Character Mesh() -> (Character Mesh:Skeletal Mesh Component Object Reference)
   Set Shoulder_rot_up_r(Shoulder_rot_up_r:double=GetAlphaFromReferenceTransformDelta(SkeletalMeshComponent=@n14, BoneName=clavicle_r, BaseName=spine_05, DeltaComponentType=EulerY, InputRange=(X=0,Y=110), TwistAxis="0,0,1")) -> (Output_Get:double)
   > Set Shoulder_rot_up_l(Shoulder_rot_up_l:double=GetAlphaFromReferenceTransformDelta(SkeletalMeshComponent=@n14, BoneName=clavicle_l, BaseName=spine_05, DeltaComponentType=EulerY, InputRange=(X=0,Y=110), TwistAxis="0,0,1")) -> (Output_Get:double)
   > Set Shoulder_rot_dn_r(Shoulder_rot_dn_r:double=GetAlphaFromReferenceTransformDelta(SkeletalMeshComponent=@n14, BoneName=upperarm_r, BaseName=clavicle_r, InputRange=(X=0,Y=-90), TwistAxis="0,0,1")) -> (Output_Get:double)
   > Set Shoulder_rot_dn_l(Shoulder_rot_dn_l:double=GetAlphaFromReferenceTransformDelta(SkeletalMeshComponent=@n14, BoneName=upperarm_l, BaseName=clavicle_l, InputRange=(X=0,Y=-90), TwistAxis="0,0,1")) -> (Output_Get:double)
   GetSkeletalMesh!(self:Custom Character Part Object Reference=Get Array Item with EFort Custom Part Type(Array:Array of Custom Character Part Object References=@n12.CustomParts) -> (Output:Custom Character Part Object Reference)) -> (ReturnValue:Skeletal Mesh Object Reference)
  ]}
]
variables [
 Body Mesh {type=SkeletalMeshComponent}
 Character Mesh {type=SkeletalMeshComponent}
 isCrouching And Moving {type=boolean}
 isDivingAndWearingCape {type=boolean}
 isWearingBackpack {type=boolean}
 isWearingCape {type=boolean}
 PlayerVector_FwdBwd {type=number}
 PlayerVector_LR {type=number}
 Shoulder_rot_dn_l {type=number}
 Shoulder_rot_dn_r {type=number}
 Shoulder_rot_up_l {type=number}
 Shoulder_rot_up_r {type=number}
]
