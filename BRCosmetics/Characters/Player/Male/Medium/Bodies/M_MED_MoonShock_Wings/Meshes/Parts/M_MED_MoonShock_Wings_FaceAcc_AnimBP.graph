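# Post-process style AnimBlueprint for the MoonShock Wings face accessory that corrects the head_fix bone as the head tilts.
# It copies the attached parent's pose and constrains head_fix to the head.
# Each update derives head_rot_up from the head-to-neck_02 rotation delta on the character mesh, cached when customization loads.
# That alpha blends in a neck-weighted constraint and an additive bone-space translation offset.
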AnimBlueprint /BRCosmetics/Characters/Player/Male/Medium/Bodies/M_MED_MoonShock_Wings/Meshes/Parts/M_MED_MoonShock_Wings_FaceAcc_AnimBP
target_skeleton=&/M_MED_MoonShock_Wings_FaceAcc_Skeleton
graphs [
 AnimGraph {nodes [
   n1 = "Constraint - head_fix (PP)"!(
    ComponentPose=Local To Component(LocalPose=Copy Pose From Mesh() {bUseAttachedParent=true})
    Alpha=1
    ConstraintWeights_0=0
    ConstraintWeights_1:float=1
   ) -> (Pose) {
    alphaBoolBlend {blendOption=Linear}
    boneToModify {boneName=head_fix}
    constraintSetup [{targetBone {boneName=neck_02}}, {targetBone {boneName=head}}]
    constraintWeights [1,1]
   }
   n2 = Get head_rot_up() -> (head_rot_up:double)
   n3 = "Constraint - head_fix (PP)"(ComponentPose=@n1, Alpha=@n2, ConstraintWeights_0=0.7, ConstraintWeights_1=0.3) {
    alphaBoolBlend {blendOption=Linear}
    boneToModify {boneName=head_fix}
    constraintSetup [{targetBone {boneName=neck_02}}, {targetBone {boneName=head}}]
    constraintWeights [1,1]
   }
   n4 = "Transform (Modify) Bone - Bone: head_fix"(ComponentPose=@n3, Alpha=@n2, Translation="3,1,0", Rotation="0,0,0", Scale="1,1,1") {
    rotationMode=BMM_Additive
    rotationSpace=BCS_BoneSpace
    scaleMode=BMM_Additive
    scaleSpace=BCS_BoneSpace
    translationMode=BMM_Additive
    translationSpace=BCS_BoneSpace
    alphaBoolBlend {blendOption=Linear}
    boneToModify {boneName=head_fix}
    translation=<3 1>
   }
   Output Pose(Result=Component To Local(ComponentPose=@n4)) {layerGroup=None}
  ]}
 EventGraph {nodes [
   Event BlueprintUpdateAnimation()
   > Set head_rot_up(head_rot_up:double=GetAlphaFromReferenceTransformDelta(
     SkeletalMeshComponent=Get Character Mesh() -> (Character Mesh:Skeletal Mesh Component Object Reference)
     BoneName=head
     BaseName=neck_02
     InputRange=(X=0,Y=-40)
     TwistAxis="0,0,1"
    )) -> (Output_Get:double)
   n1 = Event OnCharacterCustomizationLoaded()
   > n2 = Sequence!(execute) -> (then_0, then_1, then_2:Exec, then_3:Exec)
   Set Character Mesh(execute=@n2.then_1, Character Mesh:Skeletal Mesh Component Object Reference=@n1.CustomBaseMesh) -> (Output_Get:Skeletal Mesh Component Object Reference)
   Set Body Mesh(execute=@n2.then_2, Body Mesh:Skeletal Mesh Component Object Reference=Get Array Item with EFort Custom Part Type(Array=@n1.CustomMeshParts)) -> (Output_Get:Skeletal Mesh Component Object Reference)
   GetSkeletalMesh!(self:Custom Character Part Object Reference=Get Array Item with EFort Custom Part Type(Array:Array of Custom Character Part Object References=@n1.CustomParts) -> (Output:Custom Character Part Object Reference)) -> (ReturnValue:Skeletal Mesh Object Reference)
  ]}
]
variables [
 Body Mesh {type=SkeletalMeshComponent}
 Character Mesh {type=SkeletalMeshComponent}
 head_rot_up {type=number, default=1}
]
